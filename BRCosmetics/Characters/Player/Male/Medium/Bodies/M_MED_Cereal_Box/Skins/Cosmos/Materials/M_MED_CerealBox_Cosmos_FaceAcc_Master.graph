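# Master character material for the Cereal Box Cosmos skin's face accessory, adding a galaxy effect to a skeletal mesh.
# Base PreFX output feeds the Galaxy function, then passes through character PostFX and cinematic PostFX functions.
# Selectable channels of the SkinFX mask set galaxy coverage and warp.
# Swirly noise, panned over interpolated pre-skinned XZ position, drives the galaxy's alternate fresnel, keeping the pattern attached through animation.
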Material /BRCosmetics/Characters/Player/Male/Medium/Bodies/M_MED_Cereal_Box/Skins/Cosmos/Materials/M_MED_CerealBox_Cosmos_FaceAcc_Master
graph {
 properties {
  bUseMaterialAttributes=true
  bUsedWithMeshDeformer=true
  bUsedWithMorphTargets=true
  bUsedWithSkeletalMesh=true
  pixelDepthOffsetMode=PDOM_Legacy
  shadingModel=MSM_FromMaterialExpression
  subsurfaceProfile=&SS_HeroSkin_02
  usedShadingModels=Default Lit | Subsurface Profile
 }
 nodes [
  n1 = MaterialFunctionCall() {materialFunction=&BaseFN_CharacterFunction_PreFX}
  n2 = VertexInterpolator(VS=ComponentMask(None=PreSkinnedPosition()) {b=true, r=true})
  n3 = ScalarParameter() {defaultValue=1, group=Galaxy, parameterName=Edge Fresnel Tiling}
  n4 = ScalarParameter() {group=Galaxy, parameterName=GradientSpeed_X}
  n5 = ScalarParameter() {defaultValue=0.1, group=Galaxy, parameterName=GradientSpeed_Y}
  n6 = TextureSample(UVs=Panner(Coordinate=Multiply(A=@n2.PS, B=@n3), Speed=AppendVector(A=@n4, B=@n5))) {
   samplerSource=SSM_Wrap_WorldGroupSettings
   samplerType=SAMPLERTYPE_LinearColor
   texture=/Game/Weapons/WeaponSkins/Textures/T_Swirly_Noise_highContrast1
  }
  n7 = TextureObjectParameter() {group=Skin FX, parameterName=SkinFX_Mask, samplerType=SAMPLERTYPE_Masks, texture=&T_WhiteMask}
  n8 = TextureSample(Tex=@n7) {samplerType=SAMPLERTYPE_Masks}
  n9 = MaterialFunctionCall(
   AO=@n1.AO
   Distance Blend=@n1.AO
   Fresnel=@n1.AO
   Gmap=@n1.AO
   Pre Skinned Local Position=@n1.AO
   Material=MaterialFunctionCall!(
    In=@n1.Result
    Warp Mask=StaticComponentMaskParameter(None=@n8.RGBA) {defaultR=true, group=Skin FX, parameterName=Galaxy Warp Mask Channel}
    GalaxyAlternateFresnel=@n6.R
    Alpha=StaticComponentMaskParameter(None=@n8.RGBA) {defaultR=true, group=Skin FX, parameterName=Galaxy Channel}
   ) {materialFunction=/Game/Weapons/WeaponSkins/MaterialFunction/Galaxy}.Result
  ) {materialFunction=&BaseFN_CharacterFunction_PostFX}
  MaterialFunctionCall(In=@n9.Result) {materialFunction=&CINE_MF_CharPostFX}
 ]
}
